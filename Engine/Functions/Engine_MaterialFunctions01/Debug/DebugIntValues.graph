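# Material function that draws the integer digits of a scalar Number in UV space as a debug readout, ignoring sign.
# TextureCropping confines the text to an input rectangle, whose horizontal position selects a power of ten, up to a maximum digit count.
# Divide, Frac and Floor extract each digit to offset sampling into a number-strip texture.
# Masks hide leading zeros and the leftover decimal-place branch; a SphereMask dot is unconnected.
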MaterialFunction /Engine/Functions/Engine_MaterialFunctions01/Debug/DebugIntValues
graph {
 properties {description=Enter a value into the material function and read its return value as a number in uv space., libraryCategoriesText [Misc,Debug]}
 nodes [
  n1 = FunctionInput(Preview=TextureCoordinate()) {bUsePreviewValueAsDefault=true, inputName=UVs, inputType=FunctionInput_Vector2, sortPriority=5, previewValue=<7 0 0 1>}
  n2 = FunctionInput() {
   bUsePreviewValueAsDefault=true
   inputName=DebugTextLocation RG_UpperRight BA_LowerLeft
   inputType=FunctionInput_Vector4
   sortPriority=4
   previewValue=<0 0.35 1 0.65>
  }
  n3 = MaterialFunctionCall!(
   UVs=@n1
   TextureIn
   UpperLeftCorner=ComponentMask(None=@n2) {g=true, r=true}
   LowerRightCorner=ComponentMask(None=@n2) {a=true, b=true}
   ExportFloat4
  ) {materialFunction=/Engine/Functions/Engine_MaterialFunctions01/Texturing/TextureCropping}
  n4 = Multiply(A=@n3.Crop UVs, B=Constant2Vector() {g=0.1, r=2}) {constA=2, constB=0.1}
  n5 = FunctionInput() {
   bUsePreviewValueAsDefault=true
   description=Max number of displayed integers.
   inputName=MaximumNumberOfDigits
   inputType=FunctionInput_Scalar
   sortPriority=1
   previewValue=<8 0 0 1>
  }
  n6 = Multiply(A=OneMinus(None=ComponentMask(None=@n4) {r=true}), B=@n5) {constB=10}
  n7 = ComponentMask(None=@n4) {g=true}
  n8 = FunctionInput(Preview=Constant() {r=128.5}) {bUsePreviewValueAsDefault=true, inputName=Number, inputType=FunctionInput_Scalar, previewValue=<128.5 0 0 1>}
  n9 = Abs(None=@n8)
  n10 = Constant() {r=10}
  n11 = Add(A=Ceil(None=Clamp(None=@n6) {clampMode=CMODE_ClampMin, maxDefault=0}))
  n12 = Frac(None=Divide(A=@n9, B=Power(Base=@n10, Exp=@n11) {desc="10-100,000,000"}))
  n13 = Divide(A=Floor(None=Multiply(A=@n12) {constB=100})) {constB=10}
  n14 = Divide(A=@n9) {constB=1000000}
  n15 = Clamp(None=@n6) {clampMode=CMODE_ClampMax, maxDefault=0}
  n16 = Clamp(None=Subtract(A=Abs(None=@n15)) {desc=shift sampling to make room for decimal place}) {clampMode=CMODE_ClampMin, maxDefault=0}
  n17 = Floor(None=@n16)
  n18 = Frac(None=Multiply(
    A=Add(A=@n9, B=@n14) {constB=0.0001, desc="offset the floating point error precision error... "}
    B=Power(Base=@n10, Exp=@n17) {desc="10-100,000,000"}
   ))
  n19 = Divide(A=Floor(None=Multiply(A=@n18) {constB=10})) {constB=10}
  n20 = Clamp(None=Ceil(None=@n16))
  n21 = AppendVector(A=Frac(None=@n6), B=LinearInterpolate(A=Add(A=@n7, B=@n13), B=Add(A=@n7, B=@n19), Alpha=@n20))
  n22 = Power(Base=@n10, Exp=Add(A=@n11) {constB=3})
  n23 = Constant()
  n24 = Add(A=If(A=Divide(A=@n9, B=@n22), B=@n23, "A > B"=Constant() {r=1}, "A == B"=@n23, "A < B"=@n23), B=@n20)
  n25 = Ceil(None=Subtract(A=@n6) {constB=0})
  n26 = Multiply(
   A=Multiply(
    A=TextureSample!(UVs=@n21, Tex, Level, Apply View MipBias) {mipValueMode=TMVM_MipLevel, texture=/Engine/Functions/Engine_MaterialFunctions02/ExampleContent/DebugNumberStrip}.R
    B=Multiply(A=@n24, B=@n25) {desc=remove 0's on the int side}
   )
   B=@n3.Crop Mask
  ) {desc=crop mask}
  FunctionOutput(None=ComponentMask(None=Clamp(None=@n26)) {r=true})
  OneMinus(None=SphereMask(B=Constant() {r=-0.5}) {attenuationRadius=0.5, desc=mask .})
 ]
}
