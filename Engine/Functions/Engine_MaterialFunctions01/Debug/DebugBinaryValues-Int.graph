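# Debug material function that draws a scalar input, cast to int, as a row of binary digits with a configurable bit count.
# UVs split into right-to-left cells, where custom HLSL shifts the number and uses logical `&&`, not bitwise `&`, to pick 0 or 1.
# The result offsets a number-strip texture sampled with original-UV derivatives.
# A green border grid and DebugIntValues bit-index labels are added on top.
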MaterialFunction /Engine/Functions/Engine_MaterialFunctions01/Debug/DebugBinaryValues-Int
graph {
 properties {bExposeToLibrary=true, description=Displays the bit pattern of the input value as a floating point number., libraryCategoriesText [Debug]}
 nodes [
  n1 = FunctionInput(Preview=TextureCoordinate()) {bUsePreviewValueAsDefault=true, inputName=UVs, inputType=FunctionInput_Vector2, sortPriority=3}
  n2 = ComponentMask(None=Frac(None=@n1)) {g=true}
  n3 = Constant() {r=0.5}
  n4 = Ceil(None=Add(A=Distance(A=@n2, B=@n3), B=Constant() {r=-0.05}))
  MaterialExpressionFunctionInput_5 = FunctionInput() {
   bUsePreviewValueAsDefault=true
   description="Since this is the (int) version, the input will be cast to an int which means the fraction will be truncated."
   inputName=Number To Convert
   inputType=FunctionInput_Scalar
   previewValue=<32768 0 0 1>
  }
  n5 = FunctionInput() {
   bUsePreviewValueAsDefault=true
   description="The default is 32-bit since all HLSL float and int operations are 32-bit. It may be desirable to view only 8 or 16-bits when debugging custom bit packing functions. Note that this bit setting is display only, it does not affect how any calculations are done."
   inputName=Number of Bits
   inputType=FunctionInput_Scalar
   sortPriority=1
   previewValue=<32 0 0 1>
  }
  n6 = Constant2Vector() {g=-0.45}
  n7 = MaterialFunctionCall(Float2=Frac(None=Add(A=@n1, B=@n6))) {materialFunction=&BreakOutFloat2Components}
  n8 = Multiply(A=@n5, B=OneMinus(None=@n7.R)) {constB=16, desc="## bits"}
  MaterialExpressionFloor_6 = Floor(None=@n8)
  n9 = Custom!(N=@MaterialExpressionFunctionInput_5, Cell=@MaterialExpressionFloor_6) {
   code="int number = (int)N;\r\nuint curcell = (uint)Cell;\r\n\r\nuint digit = (number >> curcell) && 1;\r\n\r\n\r\nif (digit == 1) return 1;\r\n\r\nreturn 0;\r\n\r\n\r\n"
   description=Binary Digit Check
   outputType=CMOT_Float1
   inputs [
    {inputName=N, input {expression=&self:MaterialExpressionFunctionInput_5}}
    {inputName=Cell, input {expression=&self:MaterialExpressionFloor_6}}
   ]
  }
  n10 = Add(A=Multiply(A=@n9, B=Constant2Vector() {g=0.1}), B=Constant2Vector())
  n11 = Frac(None=AppendVector(A=@n8, B=@n7.G))
  n12 = Multiply(
   A=Clamp(None=OneMinus(None=@n4))
   B=TextureSample!(UVs=Add(A=@n10, B=@n11), Tex, "DDX(UVs)"=DDX(Value=@n1), "DDY(UVs)"=DDY(Value=@n1), Apply View MipBias) {mipValueMode=TMVM_Derivative, texture=/Engine/Functions/Engine_MaterialFunctions02/ExampleContent/DebugNumberStrip}.RGB
  )
  n13 = TextureSample!(UVs=@n1, Tex, Apply View MipBias) {
   desc=This just helps you see the borders so that you can tell where the ends are if the UVs tile
   texture=/Engine/EngineDebugMaterials/T_1x1_Grid
  }
  n14 = Multiply(
   A=MaterialFunctionCall!(Alpha=@n13.R, Offset=Constant() {r=0.5}) {materialFunction=/Engine/Functions/Engine_MaterialFunctions01/Opacity/AlphaOffset}.Result
   B=Constant3Vector() {constant=<rgba 0 0.2 0 0>}
  )
  n15 = Multiply(A=@MaterialExpressionFloor_6)
  n16 = ScalarParameter() {defaultValue=2, parameterName=Binary Debug Digits}
  n17 = Multiply(A=@n5)
  n18 = AppendVector(A=Multiply(A=Divide(B=@n17)) {constB=-0.2}, B=Constant() {r=-0.55})
  n19 = MaterialFunctionCall(Float2=Multiply(A=@n17, B=Add(A=@n1, B=@n18))) {materialFunction=&BreakOutFloat2Components}
  n20 = Constant() {r=0.5}
  n21 = Constant()
  n22 = Frac(None=AppendVector(A=@n19.R, B=If(A=@n19.G, B=@n20, "A > B"=@n21, "A == B"=@n19.G, "A < B"=@n21) {equalsThreshold=0.5}))
  FunctionOutput(None=Add(
    A=Add(A=@n12, B=@n14)
    B=MaterialFunctionCall!(Number=@n15, MaximumNumberOfDigits=@n16, DebugTextLocation RG_UpperRight BA_LowerLeft, UVs=@n22) {materialFunction=&/DebugIntValues}.Result
   ))
  TextureCoordinate() {uTiling=16}
  TextureCoordinate()
  TextureCoordinate()
  TextureCoordinate()
 ]
}
