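# Debug material function that sets up graph coordinates by remapping UVs into a given X/Y range, with Y flipped so the minimum sits at the bottom.
# Absolute Gradient Height times the Y span gives a debug strip thickness, padding the top when positive or the bottom when negative.
# X widens by half that thickness on each side to preserve aspect.
# It outputs the remapped coordinates, padded Y span, range and gradient height.
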MaterialFunction /Engine/Functions/Engine_MaterialFunctions01/Debug/FunctionGraphSetup
graph {nodes [
  Add()
  Add()
  MaterialExpressionFunctionInput_50 = FunctionInput() {bUsePreviewValueAsDefault=true, inputName=Gradient Height, inputType=FunctionInput_Scalar, previewValue=<0.1 0 0 1>}
  Custom!(x=@MaterialExpressionFunctionInput_50) {
   code="sign(x)"
   outputType=CMOT_Float1
   inputs [{inputName=x, input {expression=&self:MaterialExpressionFunctionInput_50}}]
  }
  n1 = FunctionInput() {
   bUsePreviewValueAsDefault=true
   description="X range=R-G, Y range =B-A"
   inputName=X range Y range
   inputType=FunctionInput_Vector4
   previewValue=<-1 1 -1 1>
  }
  n2 = ComponentMask(None=@n1) {r=true}
  n3 = ComponentMask(None=@n1) {a=true}
  n4 = ComponentMask(None=@n1) {b=true}
  n5 = Multiply(A=Abs(None=@MaterialExpressionFunctionInput_50), B=Subtract(A=@n3, B=@n4) {desc=Y size}) {desc=Calculated thickness}
  n6 = Multiply(A=@n5) {constB=0.5, desc="Add half of gradient range to X, preserves aspect ratio"}
  n7 = ComponentMask(None=@n1) {g=true}
  n8 = FunctionInput(Preview=TextureCoordinate()) {
   bUsePreviewValueAsDefault=true
   description=Leave this input alone unless you want to change index or know the exact effect of remapping when the input range is not 0-1.
   inputName=UVs
   inputType=FunctionInput_Vector2
   sortPriority=1
   previewValue=<0 1 -1 1>
  }
  n9 = LinearInterpolate(A=Subtract(A=@n2, B=@n6), B=Add(A=@n7, B=@n6), Alpha=ComponentMask(None=@n8) {r=true}) {constA=-1, desc=X}
  n10 = Constant()
  n11 = If(A=@MaterialExpressionFunctionInput_50, B=@n10, "A > B"=Add(A=@n5, B=@n3) {desc=Pad Max by debug thickness}, "A == B"=@n3, "A < B"=@n3) {desc="Correct Max, A"}
  n12 = If(A=@MaterialExpressionFunctionInput_50, B=@n10, "A > B"=@n4, "A == B"=@n4, "A < B"=Subtract(A=@n4, B=@n5) {desc=Pad Min by debug thickness}) {desc="Correct Min, B"}
  n13 = LinearInterpolate(A=@n11, B=@n12, Alpha=ComponentMask(None=@n8) {g=true}) {constA=-1, desc="Y, Flip so 0 is on bottom"}
  FunctionOutput(None=AppendVector(A=@n9, B=@n13)) {outputName=UV, sortPriority=1}
  FunctionOutput(None=@n9) {outputName=x, sortPriority=10}
  FunctionOutput(None=@MaterialExpressionFunctionInput_50) {outputName=Gradient Height, sortPriority=7}
  FunctionOutput(None=@n13) {outputName=y, sortPriority=11}
  FunctionOutput(None=Subtract(A=@n11, B=@n12)) {outputName=Corrected Y, sortPriority=6}
  FunctionOutput(None=@n1) {outputName=Range}
  Subtract(A=@n7, B=@n2) {desc=X size}
 ]}
