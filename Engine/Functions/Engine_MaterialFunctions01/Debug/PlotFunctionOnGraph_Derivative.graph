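# Debug material function that plots f(x) = y over given X and Y ranges, with a finite-difference derivative.
# It outputs x and x + Δx for the plotted function, which the caller evaluates twice and feeds back.
# The aspect-corrected slope is output as d(x) and keeps line thickness even.
# A static switch adds a clipped grid and gradient strip; the plot fades with camera depth.
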MaterialFunction /Engine/Functions/Engine_MaterialFunctions01/Debug/PlotFunctionOnGraph_Derivative
graph {
 properties {
  bExposeToLibrary=true
  description="Plots f(x) = y to visualize the output of another function. This version includes the derivative which requires running the function twice, once with  x + Δx as the input."
 }
 nodes [
  n1 = FunctionInput() {
   bUsePreviewValueAsDefault=true
   description="X range=R-G, Y range =B-A"
   inputName=X range Y range
   inputType=FunctionInput_Vector4
   previewValue=<-1 1 -1 1>
  }
  n2 = ComponentMask(None=@n1) {g=true}
  n3 = ComponentMask(None=@n1) {r=true}
  n4 = ComponentMask(None=@n1) {a=true}
  n5 = ComponentMask(None=@n1) {b=true}
  n6 = Subtract(A=@n4, B=@n5) {desc=Y size}
  MaterialExpressionDivide_64 = Divide(A=Subtract(A=@n2, B=@n3) {desc=X size}, B=@n6) {desc=Aspect Ratio}
  n7 = FunctionInput() {bUsePreviewValueAsDefault=true, inputName="f(x + Δx)", inputType=FunctionInput_Scalar, sortPriority=6}
  n8 = FunctionInput() {inputName="f(x)", inputType=FunctionInput_Scalar, sortPriority=4}
  n9 = Constant() {desc=Δx, r=0.0001}
  n10 = Divide(A=Subtract(A=@n7, B=@n8), B=@n9) {desc=Derivative}
  n11 = If(A=@n7, B=Constant(), "A > B"=@n10, "A == B"=Constant(), "A < B"=@n10)
  MaterialExpressionMultiply_123 = Multiply(A=@MaterialExpressionDivide_64, B=@n11) {desc="Slope * Aspect Ratio = New slope, yay this worked first guess"}
  Add(B=@MaterialExpressionMultiply_123)
  n12 = Frac(None=ComponentMask() {r=true})
  n13 = Frac(None=ComponentMask() {g=true})
  Add(A=Floor(None=Multiply(A=@n12) {constB=1.05}), B=Floor(None=Multiply(A=@n13) {constB=1.05}))
  MaterialExpressionFunctionInput_49 = FunctionInput() {bUsePreviewValueAsDefault=true, inputName=Gradient Height, inputType=FunctionInput_Scalar, previewValue=<0.1 0 0 1>}
  n14 = Constant()
  n15 = Abs(None=@MaterialExpressionFunctionInput_49)
  n16 = Multiply(A=@n15, B=@n6) {desc=Calculated thickness}
  n17 = If(A=@MaterialExpressionFunctionInput_49, B=@n14, "A > B"=Add(A=@n16, B=@n4) {desc=Pad Max by debug thickness}, "A == B"=@n4, "A < B"=@n4) {desc="Correct Max, A"}
  n18 = If(A=@MaterialExpressionFunctionInput_49, B=@n14, "A > B"=@n5, "A == B"=@n5, "A < B"=Subtract(A=@n5, B=@n16) {desc=Pad Min by debug thickness}) {desc="Correct Min, B"}
  n19 = Divide(B=Subtract(A=@n17, B=@n18) {desc=Corrected Y size}) {desc=Y Ratio}
  Add(A=Add(A=MaterialFunctionCall(Number=@n19) {materialFunction=&DebugScalarValues}.Result) {desc=Debug})
  Clamp(None=Divide(B=OneMinus(None=@MaterialExpressionMultiply_123))) {desc="This hack prevents crazy spikes on the graph end points. "}
  n20 = Custom!(x=@MaterialExpressionDivide_64) {
   code="atan(x)"
   outputType=CMOT_Float1
   inputs [{
     inputName=x
     input {expression=&self:MaterialExpressionDivide_64}
    }]
  }
  MaterialExpressionDivide_5 = Divide(A=Subtract(B=Multiply(A=@n20) {constB=57.2958, desc=Convert to Degrees}) {constA=90}) {constB=57.2958, desc=Back ro radians}
  n21 = Custom(x=@MaterialExpressionDivide_5) {
   code="sin(x)"
   desc=How long is line
   outputType=CMOT_Float1
   inputs [{inputName=x, input {expression=&self:MaterialExpressionDivide_5}}]
  }
  Clamp(None=@n21) {desc="This hack prevents crazy spikes on the graph end points. ", minDefault=0.2}
  Constant3Vector() {constant=<rgba 0 0.3084 0.5 0>}
  n22 = Custom(x=@MaterialExpressionMultiply_123) {
   code="atan(x)"
   outputType=CMOT_Float1
   inputs [{
     inputName=x
     input {expression=&self:MaterialExpressionMultiply_123}
    }]
  }
  MaterialExpressionDivide_4 = Divide(A=Subtract(B=Multiply(A=@n22) {constB=57.2958, desc=Convert to Degrees}) {constA=90}) {constB=57.2958, desc=Back ro radians}
  Custom(x=@MaterialExpressionDivide_4) {
   code="sin(x)"
   desc=How long is line
   outputType=CMOT_Float1
   inputs [{inputName=x, input {expression=&self:MaterialExpressionDivide_4}}]
  }
  Custom(x=@MaterialExpressionDivide_4) {
   code="tan(x)"
   desc=How long is line
   outputType=CMOT_Float1
   inputs [{inputName=x, input {expression=&self:MaterialExpressionDivide_4}}]
  }
  n23 = Constant() {r=1000}
  n24 = MaterialFunctionCall(Fade Length=Constant() {r=6500}, Fade Offset=@n23) {materialFunction=&CameraDepthFade}
  n25 = Multiply(A=@n15) {constB=0.5, desc="Add half of gradient range to X, preserves aspect ratio"}
  n26 = TextureCoordinate()
  n27 = LinearInterpolate(A=Subtract(A=@n3, B=@n25), B=Add(A=@n2, B=@n25), Alpha=ComponentMask(None=@n26) {r=true}) {constA=-1, desc=X}
  n28 = Multiply(A=@n27)
  n29 = Multiply(A=@n28) {constB=-1}
  n30 = Multiply(A=Ceil(None=Subtract(A=@n28, B=@n3)), B=Ceil(None=Add(A=@n29, B=@n2)))
  n31 = AppendVector(A=@n27, B=LinearInterpolate(A=@n17, B=@n18, Alpha=ComponentMask(None=@n26) {g=true}) {constA=-1, desc="Y, Flip so 0 is on bottom"})
  n32 = Multiply(A=@n31) {constA=1.001, constB=1.001}
  n33 = Custom(x=@MaterialExpressionFunctionInput_49) {
   code="sign(x)"
   outputType=CMOT_Float1
   inputs [{
     inputName=x
     input {expression=&self:MaterialExpressionFunctionInput_49}
    }]
  }
  n34 = ComponentMask(None=@n31) {g=true}
  n35 = Subtract(A=Multiply(A=@n33, B=@n34), B=Multiply(A=@n4))
  n36 = LinearInterpolate(
   A=Multiply(A=TextureSample(UVs=@n32) {texture=/Engine/EngineDebugMaterials/T_1x1_Grid}.RGB, B=Constant3Vector() {constant=<rgba 0 0.03 0.25 0>})
   B=@n8
   Alpha=Clamp(None=Ceil(None=@n35))
  ) {desc=Gradient Result}
  n37 = Distance(A=ComponentMask(None=@n31) {r=true}, B=@n34)
  n38 = FunctionInput() {bUsePreviewValueAsDefault=true, inputName=Thickness, inputType=FunctionInput_Scalar, sortPriority=2, previewValue=<1 0 0 1>}
  n39 = Divide(A=Divide(A=@n38) {constB=400}, B=@n19)
  n40 = Floor(None=Add(A=OneMinus(None=@n37), B=Multiply(A=@n39) {constB=2}) {constB=0.01})
  n41 = OneMinus(None=Distance(A=@n34, B=@n8))
  n42 = Multiply(A=@MaterialExpressionMultiply_123, B=@MaterialExpressionMultiply_123)
  n43 = Multiply(A=@n39, B=SquareRoot(None=Add(A=@n42)))
  n44 = Clamp(None=Floor(None=Add(A=@n41, B=@n43) {constB=0.01}))
  n45 = FunctionInput(Preview=Constant3Vector() {constant=<rgba 2 0 0 0>}) {bUsePreviewValueAsDefault=true, inputName=Color, sortPriority=3}
  n46 = OneMinus(None=Distance(A=@n34, B=@n11))
  n47 = Clamp(None=Floor(None=Add(A=@n46, B=@n39) {constB=0.01}))
  n48 = Add(A=Multiply(A=Clamp(None=@n40)) {constB=0.25}, B=Add(A=Multiply(A=@n44, B=@n45), B=Multiply(A=@n47) {constB=0}))
  n49 = StaticSwitch(
   True=Multiply(A=Clamp(None=@n30), B=Add(A=@n36, B=@n48) {constB=0, desc=Grid}) {desc=Clip grid at range minmax}
   False=@n48
   Value=FunctionInput(Preview=StaticBool() {value=true}) {bUsePreviewValueAsDefault=true, inputName=ShowGrid?, inputType=FunctionInput_StaticBool, sortPriority=2}
  )
  FunctionOutput(None=Multiply(A=LinearInterpolate(Alpha=@n24.Result) {constA=1, constB=0}, B=Clamp(None=@n49) {maxDefault=10})) {outputName="f(x) = y"}
  FunctionOutput(None=@MaterialExpressionMultiply_123) {outputName="d(x)"}
  FunctionOutput(None=@n27) {outputName=x, sortPriority=1}
  FunctionOutput(None=Add(A=@n27, B=@n9) {desc=+ Δx}) {outputName=x + Δx, sortPriority=2}
  n50 = Constant() {r=0.005}
  n51 = Constant() {r=0.05}
  n52 = MaterialFunctionCall(Fade Length=Constant() {r=3500}, Fade Offset=@n23) {materialFunction=&CameraDepthFade}
  n53 = Divide(A=Divide(A=LinearInterpolate(A=@n50, B=@n51, Alpha=@n52.Result), B=@n19) {desc=/ y-ratio}) {constB=1}
  MaterialFunctionCall!(Alpha=@n44, Offset=OneMinus(None=Clamp(None=@n53))) {materialFunction=/Engine/Functions/Engine_MaterialFunctions01/Opacity/AlphaOffset}
 ]
}
