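# Debug material function that plots a scalar f(x) as a colored line over a chosen X/Y range to visualize another function's output.
# Pixels near f(x) are lit via distance and Floor thresholding, plus a faint y=x diagonal.
# ShowGrid adds a tinted grid, an X-range mask and a grayscale f(x) strip sized by Gradient Height.
# Default UVs remap texture coordinates, flipping Y and padding for the strip; slope-based thickness correction is left unconnected.
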MaterialFunction /Engine/Functions/Engine_MaterialFunctions01/Debug/PlotFunctionOnGraph_Setup_Input
graph {
 properties {bExposeToLibrary=true, description="Plots f(x) = y to visualize the output of another function."}
 nodes [
  n1 = Frac(None=ComponentMask() {r=true})
  n2 = Frac(None=ComponentMask() {g=true})
  Add(A=Floor(None=Multiply(A=@n1) {constB=1.05}), B=Floor(None=Multiply(A=@n2) {constB=1.05}))
  n3 = FunctionInput() {
   bUsePreviewValueAsDefault=true
   description="X range=R-G, Y range =B-A"
   inputName=X range Y range
   inputType=FunctionInput_Vector4
   previewValue=<-1 1 -1 1>
  }
  n4 = ComponentMask(None=@n3) {r=true}
  MaterialExpressionFunctionInput_48 = FunctionInput() {bUsePreviewValueAsDefault=true, inputName=Gradient Height, inputType=FunctionInput_Scalar, sortPriority=5, previewValue=<0.1 0 0 1>}
  n5 = ComponentMask(None=@n3) {a=true}
  n6 = ComponentMask(None=@n3) {b=true}
  n7 = Subtract(A=@n5, B=@n6) {desc=Y size}
  n8 = Multiply(A=Abs(None=@MaterialExpressionFunctionInput_48), B=@n7) {desc=Calculated thickness}
  n9 = Multiply(A=@n8) {constB=0.5, desc="Add half of gradient range to X, preserves aspect ratio"}
  n10 = Subtract(A=@n4, B=@n9)
  n11 = ComponentMask(None=@n3) {g=true}
  n12 = Add(A=@n11, B=@n9)
  n13 = TextureCoordinate()
  n14 = ComponentMask(None=@n13) {r=true}
  n15 = Constant()
  n16 = If(A=@MaterialExpressionFunctionInput_48, B=@n15, "A > B"=Add(A=@n8, B=@n5) {desc=Pad Max by debug thickness}, "A == B"=@n5, "A < B"=@n5) {desc="Correct Max, A"}
  n17 = If(A=@MaterialExpressionFunctionInput_48, B=@n15, "A > B"=@n6, "A == B"=@n6, "A < B"=Subtract(A=@n6, B=@n8) {desc=Pad Min by debug thickness}) {desc="Correct Min, B"}
  n18 = ComponentMask(None=@n13) {g=true}
  n19 = FunctionInput(Preview=AppendVector(
    A=LinearInterpolate(A=@n10, B=@n12, Alpha=@n14) {constA=-1, desc=X}
    B=LinearInterpolate(A=@n16, B=@n17, Alpha=@n18) {constA=-1, desc="Y, Flip so 0 is on bottom"}
   )) {inputName=UVs, inputType=FunctionInput_Vector2, sortPriority=1}
  n20 = ComponentMask(None=@n19) {g=true}
  n21 = FunctionInput() {inputName="f(x)", inputType=FunctionInput_Scalar, sortPriority=20}
  n22 = FunctionInput() {bUsePreviewValueAsDefault=true, inputName=Thickness, inputType=FunctionInput_Scalar, sortPriority=7, previewValue=<1 0 0 1>}
  n23 = Subtract(A=@n16, B=@n17) {desc=Corrected Y size}
  n24 = Divide(
   A=Divide(A=@n22) {constB=200}
   B=Divide(B=FunctionInput(Preview=@n23) {inputName=Corrected Y-size, inputType=FunctionInput_Scalar, sortPriority=4}) {desc=Y Ratio}
  )
  n25 = Add(A=OneMinus(None=Distance(A=@n20, B=@n21)), B=@n24) {constB=0.01}
  n26 = Multiply(
   A=Clamp(None=Floor(None=@n25))
   B=FunctionInput(Preview=Constant3Vector() {constant=<rgba 2 0 0 0>}) {bUsePreviewValueAsDefault=true, inputName=Color, sortPriority=9}
  )
  Add(A=@n26)
  MaterialExpressionMultiply_141 = Multiply(A=Divide(A=Subtract(A=@n11, B=@n4) {desc=X size}, B=@n7) {desc=Aspect Ratio}, B=Divide(A=Constant() {r=1}) {constB=1, desc=Derivative}) {desc="Slope * Aspect Ratio = New slope, yay this worked first guess"}
  n27 = Custom!(x=@MaterialExpressionMultiply_141) {
   code="atan(x)"
   outputType=CMOT_Float1
   inputs [{
     inputName=x
     input {expression=&self:MaterialExpressionMultiply_141}
    }]
  }
  MaterialExpressionDivide_71 = Divide(A=Subtract(B=Multiply(A=@n27) {constB=57.2958, desc=Convert to Degrees}) {constA=90}) {constB=57.2958, desc=Back ro radians}
  Custom(x=@MaterialExpressionDivide_71) {
   code="tan(x)"
   desc=How long is line
   outputType=CMOT_Float1
   inputs [{
     inputName=x
     input {expression=&self:MaterialExpressionDivide_71}
    }]
  }
  n28 = Custom(x=@MaterialExpressionDivide_71) {
   code="sin(x)"
   desc=How long is line
   outputType=CMOT_Float1
   inputs [{
     inputName=x
     input {expression=&self:MaterialExpressionDivide_71}
    }]
  }
  Divide(A=@n24, B=Clamp(None=@n28) {desc="This hack prevents crazy spikes on the graph end points. ", minDefault=0.2}) {constB=0.707, desc=Modified Thickness}
  n29 = ComponentMask(None=@n19) {r=true}
  n30 = Multiply(A=@n29)
  n31 = Subtract(A=@n30, B=@n4)
  n32 = Add(A=Multiply(A=@n30) {constB=-1}, B=@n11)
  n33 = Clamp(None=Multiply(A=Ceil(None=@n31), B=Ceil(None=@n32)))
  n34 = TextureSample(UVs=Multiply(A=@n19) {constA=1.001, constB=1.001}) {texture=/Engine/EngineDebugMaterials/T_1x1_Grid}
  n35 = Constant3Vector() {constant=<rgba 0 0.02 0.15 0>}
  n36 = Custom(x=@MaterialExpressionFunctionInput_48) {
   code="sign(x)"
   outputType=CMOT_Float1
   inputs [{
     inputName=x
     input {expression=&self:MaterialExpressionFunctionInput_48}
    }]
  }
  n37 = Ceil(None=Subtract(A=Multiply(A=@n36, B=@n20), B=Multiply(A=@n5)))
  n38 = Divide(A=Distance(A=@n29, B=@n20)) {constB=1.41}
  n39 = Floor(None=Add(A=OneMinus(None=@n38), B=@n24) {constB=0.01})
  n40 = Add(A=Multiply(A=Clamp(None=@n39)) {constB=0.1}, B=@n26)
  n41 = Add(A=LinearInterpolate(A=Multiply(A=@n34.RGB, B=@n35), B=@n21, Alpha=Clamp(None=@n37)), B=@n40)
  n42 = StaticBool() {value=true}
  FunctionOutput(None=StaticSwitch(
    True=Multiply(A=@n33, B=@n41)
    False=@n40
    Value=FunctionInput(Preview=@n42) {bUsePreviewValueAsDefault=true, inputName=ShowGrid?, inputType=FunctionInput_StaticBool, sortPriority=10}
   )) {outputName="f(x) = y"}
  Multiply()
  Subtract(B=@n21)
 ]
}
